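# Third-person sound cue for the Pirate Gold emote music, playing a single mixed wave through one wave player.
# It uses Athena emote attenuation and the third-person emote music sound class, and keeps playing when silent through virtualization.
# It references the shared emote music concurrency rules and also sends to the music audio-analysis submix.
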SoundCue /BRCosmetics/Sounds/Emotes/Pirate_Gold/SC_Emote_Pirate_Gold_Mix_3P_Cue
root=WavePlayer_0
nodes [{id=WavePlayer_0, properties {soundWaveAssetPtr=&/Emote_Pirate_Gold_Mix_01}}]
properties {
 attenuationSettings=&Fort_Emotes_Athena_Att
 maxDistance=1600
 soundClassObject=/Game/Sounds/FortSoundClassesAndModes/Fort_Music_Emote_3P
 virtualizationMode=PlayWhenSilent
 volumeMultiplier=1
 concurrencyOverrides {resolutionRule=PreventNew}
 concurrencySet [/Game/Sounds/Concurrency_Rules/Athena_EmoteMusic_Conc]
 soundSubmixSends [{minSendLevel=0, soundSubmix=&Fortnite_AudioAnalysis_Music}]
}
